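# Material function that decodes a vertex animation normal texture from AnimToTexture into a tangent-space normal.
# It samples the input texture at the input UVs as linear color and remaps RGB from 0..1 to -1..1.
# The result is transformed from local to tangent space and passed through a VertexInterpolator, so the lookup runs per vertex and reaches the pixel shader interpolated.
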MaterialFunction /AnimToTexture/Materials/MaterialFunctions/UnpackVertexNormal
graph {nodes [
  n1 = FunctionInput() {inputName=UVs, inputType=FunctionInput_Vector2}
  n2 = FunctionInput(Preview=TextureObject()) {inputName=Tex, inputType=FunctionInput_Texture2D, sortPriority=1}
  n3 = Subtract(A=Multiply(A=TextureSample(UVs=@n1, Tex=@n2) {samplerType=SAMPLERTYPE_LinearColor, texture=/AnimToTexture/Characters/Mannequin/Textures/VertexAnimation/TX_VertexNormal}.RGB) {constB=2})
  FunctionOutput(None=VertexInterpolator(VS=Transform(None=@n3) {transformSourceType=TRANSFORMSOURCE_Local, transformType=TRANSFORM_Tangent}).PS) {outputName=Normal}
 ]}
