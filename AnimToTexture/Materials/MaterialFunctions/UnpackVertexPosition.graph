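# Material function for AnimToTexture that decodes a vertex position stored in a texture into a World Position Offset.
# It samples the texture at the input UVs as linear color, then scales RGB by SizeBBox and adds MinBBox to unpack the bounding-box-normalized value.
# The result is transformed to world space, and the Object Pivot Location is subtracted to produce the offset.
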MaterialFunction /AnimToTexture/Materials/MaterialFunctions/UnpackVertexPosition
graph {nodes [
  n1 = FunctionInput() {inputName=UVs, inputType=FunctionInput_Vector2}
  n2 = FunctionInput(Preview=TextureObject()) {inputName=Tex, inputType=FunctionInput_Texture2D, sortPriority=1}
  n3 = Add(
   A=Multiply(A=TextureSample(UVs=@n1, Tex=@n2) {samplerType=SAMPLERTYPE_LinearColor}.RGB, B=FunctionInput() {inputName=SizeBBox, sortPriority=2}) {constB=1000}
   B=FunctionInput() {inputName=MinBBox, sortPriority=3}
  )
  FunctionOutput(None=Subtract(
    A=TransformPosition(None=@n3) {transformType=TRANSFORMPOSSOURCE_World}
    B=MaterialFunctionCall() {materialFunction=&ObjectPivotPoint}.Object Pivot Location
   )) {outputName=WorldPositionOffset}
 ]}
