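# AnimToTexture material function that outputs one of the mesh UV channels 0 to 3, selected at compile time.
# Nested static switch parameters UseUV0 through UseUV3, grouped under AnimToTexture, each pick their matching TextureCoordinate node.
# The lowest-numbered enabled switch wins, and UV0 is the fallback when none are enabled.
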MaterialFunction /AnimToTexture/Materials/MaterialFunctions/TexCoord
graph {nodes [
  n1 = TextureCoordinate() {coordinateIndex=1}
  n2 = StaticSwitchParameter(
   True=TextureCoordinate() {coordinateIndex=2}
   False=StaticSwitchParameter(True=TextureCoordinate() {coordinateIndex=3}, False=TextureCoordinate()) {group=AnimToTexture, parameterName=UseUV3, sortPriority=10}
  ) {group=AnimToTexture, parameterName=UseUV2, sortPriority=10}
  FunctionOutput(None=StaticSwitchParameter(True=TextureCoordinate(), False=StaticSwitchParameter(True=@n1, False=@n2) {group=AnimToTexture, parameterName=UseUV1, sortPriority=10}) {group=AnimToTexture, parameterName=UseUV0, sortPriority=10})
 ]}
