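# VertexUVs is an AnimToTexture material function that computes the UVs for a given animation frame in a texture.
# It floors Frame and RowsPerFrame, multiplies them, divides by the texture height and applies Fmod by that height.
# The result shifts V of the UVs from the TexCoord function, leaving U unchanged.
# An unconnected Custom HLSL node holds an equivalent version of this math.
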MaterialFunction /AnimToTexture/Materials/MaterialFunctions/VertexUVs
graph {nodes [
  Custom!(UVs, Frame, TextureSize, RowsPerFrame) {
   code="UVs.y += ((int(Frame) * int(RowsPerFrame) / TextureSize.y)) % TextureSize.y;\r\n\r\nreturn UVs;"
   desc=This can be done with Math nodes
   description=OffsetUVs
   outputType=CMOT_Float2
   inputs [
    {inputName=UVs, input {expression=null}}
    {inputName=Frame, input {expression=null}}
    {inputName=TextureSize, input {expression=null}}
    {inputName=RowsPerFrame, input {expression=null}}
   ]
  }
  n1 = MaterialFunctionCall(Float2=MaterialFunctionCall() {materialFunction=&/TexCoord}.Result) {materialFunction=&BreakOutFloat2Components}
  n2 = Floor(None=FunctionInput() {inputName=Frame, inputType=FunctionInput_Scalar})
  n3 = Floor(None=FunctionInput() {inputName=RowsPerFrame, inputType=FunctionInput_Scalar, sortPriority=3})
  n4 = MaterialFunctionCall(Float2=FunctionInput() {inputName=TextureSize, inputType=FunctionInput_Vector2, sortPriority=2}) {materialFunction=&BreakOutFloat2Components}
  n5 = Fmod(A=Divide(A=Multiply(A=@n2, B=@n3), B=@n4.G), B=@n4.G)
  FunctionOutput(None=MaterialFunctionCall(X=@n1.R, Y=Add(A=@n1.G, B=@n5)) {materialFunction=&MakeFloat2}.Result) {outputName=UVs}
 ]}
