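# Material function for AnimToTexture that reads baked vertex animation textures and outputs a vertex position offset and normal for a given frame.
# VertexUVs derives lookup UVs from the frame, rows per frame and the position texture's size.
# Those shared UVs feed UnpackVertexPosition, which uses the bounding box min and size to produce world position offset, and UnpackVertexNormal, which samples the normals texture.
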MaterialFunction /AnimToTexture/Materials/MaterialFunctions/VertexPositionAndNormal
graph {nodes [
  n1 = FunctionInput(Preview=TextureObject()) {inputName=VertexPositionsTexture, inputType=FunctionInput_Texture2D, sortPriority=3}
  n2 = MaterialFunctionCall!(
   Frame=FunctionInput() {inputName=Frame, inputType=FunctionInput_Scalar}
   TextureSize=TextureProperty(Tex=@n1)
   RowsPerFrame=FunctionInput() {inputName=RowsPerFrame, inputType=FunctionInput_Scalar, sortPriority=1}
  ) {materialFunction=&/VertexUVs}
  FunctionOutput(None=MaterialFunctionCall!(
    UVs=@n2.UVs
    Tex=@n1
    SizeBBox=FunctionInput() {inputName=SizeBBox, sortPriority=5}
    MinBBox=FunctionInput() {inputName=MinBBox, sortPriority=6}
   ) {materialFunction=&/UnpackVertexPosition}.WorldPositionOffset) {outputName=Position, sortPriority=1}
  n3 = TextureObject()
  FunctionOutput(None=MaterialFunctionCall!(UVs=@n2.UVs, Tex=FunctionInput(Preview=@n3) {inputName=VertexNormalsTexture, inputType=FunctionInput_Texture2D, sortPriority=4}) {materialFunction=&/UnpackVertexNormal}.Normal) {outputName=Normal}
 ]}
