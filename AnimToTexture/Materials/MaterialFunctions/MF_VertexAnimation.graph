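# Material function that adds AnimToTexture vertex animation to incoming material attributes, playing baked mesh animation from textures.
# GetFrameSwitch supplies the frame to VertexPositionAndNormal, alongside position and normal texture parameters, rows per frame and bounding-box min and size.
# SetMaterialAttributes overrides Normal and World Position Offset, both wired to that function's Normal output as shown, then outputs the attributes.
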MaterialFunction /AnimToTexture/Materials/MaterialFunctions/MF_VertexAnimation
graph {nodes [
  n1 = VectorParameter() {group=AnimToTexture, parameterName=MinBBox, sortPriority=20, defaultValue=<rgba -70.7 -41.4 -141.4 0>}
  n2 = MaterialFunctionCall!(
   Frame=MaterialFunctionCall() {materialFunction=&/GetFrameSwitch}.Result
   RowsPerFrame=ScalarParameter() {defaultValue=1, group=AnimToTexture, parameterName=RowsPerFrame, sliderMin=1, sortPriority=20}
   VertexPositionsTexture=TextureObjectParameter() {group=AnimToTexture, parameterName=PositionTexture, sortPriority=30}
   VertexNormalsTexture=TextureObjectParameter() {group=AnimToTexture, parameterName=NormalTexture, sortPriority=30}
   SizeBBox=VectorParameter() {group=AnimToTexture, parameterName=SizeBBox, sortPriority=20}.RGB
   MinBBox=@n1.RGB
  ) {materialFunction=&/VertexPositionAndNormal}
  FunctionOutput(None=SetMaterialAttributes!(
    MaterialAttributes=FunctionInput() {inputName=Material Attributes, inputType=FunctionInput_MaterialAttributes}
    Normal=@n2.Normal
    WorldPositionOffset=@n2.Normal
   ) {attributeSetTypes [#1, #2]})
 ]}
