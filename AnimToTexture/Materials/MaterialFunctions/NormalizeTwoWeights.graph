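# Material function in the AnimToTexture folder that normalizes two scalar weights so they sum to one.
# It computes the reciprocal of Weight1 plus Weight2 once and multiplies each input by it to produce the two outputs.
# There is no guard against a zero sum.
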MaterialFunction /AnimToTexture/Materials/MaterialFunctions/NormalizeTwoWeights
graph {nodes [
  n1 = FunctionInput() {inputName=Weight1, inputType=FunctionInput_Scalar}
  n2 = FunctionInput() {inputName=Weight2, inputType=FunctionInput_Scalar, sortPriority=1}
  n3 = Divide(A=Constant() {r=1}, B=Add(A=@n1, B=@n2))
  FunctionOutput(None=Multiply(A=@n1, B=@n3)) {outputName=Weight1}
  FunctionOutput(None=Multiply(A=@n3, B=@n2)) {outputName=Weight2, sortPriority=1}
 ]}
