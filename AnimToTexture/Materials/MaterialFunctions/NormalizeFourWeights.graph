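# NormalizeFourWeights is an AnimToTexture material function that rescales four scalar weights so they sum to one.
# It adds the four inputs, takes the reciprocal of the total with a single divide, then multiplies each input by that shared factor to give four outputs.
# The sum has no zero guard, so all-zero inputs cause a division by zero.
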MaterialFunction /AnimToTexture/Materials/MaterialFunctions/NormalizeFourWeights
graph {nodes [
  n1 = FunctionInput() {inputName=Weight1, inputType=FunctionInput_Scalar}
  n2 = FunctionInput() {inputName=Weight2, inputType=FunctionInput_Scalar, sortPriority=1}
  n3 = FunctionInput() {inputName=Weight3, inputType=FunctionInput_Scalar, sortPriority=2}
  n4 = FunctionInput() {inputName=Weight4, inputType=FunctionInput_Scalar, sortPriority=3}
  n5 = Divide(A=Constant() {r=1}, B=Add(A=Add(A=@n1, B=@n2), B=Add(A=@n3, B=@n4)))
  FunctionOutput(None=Multiply(A=@n1, B=@n5)) {outputName=Weight1}
  FunctionOutput(None=Multiply(A=@n5, B=@n2)) {outputName=Weight2, sortPriority=1}
  FunctionOutput(None=Multiply(A=@n5, B=@n3)) {outputName=Weight3, sortPriority=2}
  FunctionOutput(None=Multiply(A=@n5, B=@n4)) {outputName=Weight4, sortPriority=3}
 ]}
