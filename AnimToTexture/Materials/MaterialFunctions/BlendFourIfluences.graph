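# Material function that blends four influences, each a position, normal and weight, into one Position and one Normal, as part of AnimToTexture.
# The four weight inputs first pass through the NormalizeFourWeights function.
# Position is the weighted sum of the four input positions.
# Normal is the weighted sum of the four input normals, renormalized to unit length.
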MaterialFunction /AnimToTexture/Materials/MaterialFunctions/BlendFourIfluences
graph {nodes [
  n1 = MaterialFunctionCall!(
   Weight1=FunctionInput() {inputName=Weight1, inputType=FunctionInput_Scalar, sortPriority=2}
   Weight2=FunctionInput() {inputName=Weight2, inputType=FunctionInput_Scalar, sortPriority=5}
   Weight3=FunctionInput() {inputName=Weight3, sortPriority=8}
   Weight4=FunctionInput() {inputName=Weight4, sortPriority=11}
  ) {materialFunction=&/NormalizeFourWeights}
  n2 = Add(
   A=Multiply(A=@n1.Weight1, B=FunctionInput() {inputName=Normal1, sortPriority=1})
   B=Multiply(A=@n1.Weight2, B=FunctionInput() {inputName=Normal2, sortPriority=4})
  )
  n3 = Add(
   A=Multiply(A=@n1.Weight3, B=FunctionInput() {inputName=Normal3, sortPriority=7})
   B=Multiply(A=@n1.Weight4, B=FunctionInput() {inputName=Normal4, sortPriority=10})
  )
  FunctionOutput(None=Normalize(VectorInput=Add(A=@n2, B=@n3))) {outputName=Normal}
  n4 = Multiply(A=FunctionInput() {inputName=Position1}, B=@n1.Weight1)
  n5 = Multiply(A=FunctionInput() {inputName=Position2, sortPriority=3}, B=@n1.Weight2)
  n6 = Multiply(A=FunctionInput() {inputName=Position3, sortPriority=6}, B=@n1.Weight3)
  n7 = Multiply(A=FunctionInput() {inputName=Position4, sortPriority=9}, B=@n1.Weight4)
  FunctionOutput(None=Add(A=Add(A=@n4, B=@n5), B=Add(A=@n6, B=@n7))) {outputName=Position, sortPriority=1}
 ]}
