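# BoneIndexUVs is an AnimToTexture material function that converts a frame, a bone index and a texture size into UVs for addressing a texel.
# It floors both scalars and divides by texture size, mapping bone index to U and frame to V.
# A half-texel offset on each axis targets texel centers.
# An unconnected Custom HLSL node holds similar math using int casts.
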MaterialFunction /AnimToTexture/Materials/MaterialFunctions/BoneIndexUVs
graph {nodes [
  Custom!(Frame, BoneIndex, TextureSize) {
   code="float2 UV;\r\n\r\nUV.x = (0.5 / TextureSize.x) + (int(BoneIndex) / TextureSize.x);\r\nUV.y = (0.5 / TextureSize.y) + (int(Frame) / TextureSize.y);\r\n\r\nreturn UV;"
   outputType=CMOT_Float2
   inputs [{inputName=Frame, input {expression=null}}, {inputName=BoneIndex, input {expression=null}}, {inputName=TextureSize, input {expression=null}}]
  }
  n1 = MaterialFunctionCall(Float2=FunctionInput() {inputName=TextureSize, inputType=FunctionInput_Vector2, sortPriority=2}) {materialFunction=&BreakOutFloat2Components}
  n2 = Divide(A=Floor(None=FunctionInput() {inputName=BoneIndex, inputType=FunctionInput_Scalar, sortPriority=1}), B=@n1.R)
  n3 = Divide(B=@n1.R) {constA=0.5}
  n4 = Divide(B=@n1.G) {constA=0.5}
  n5 = Divide(A=Floor(None=FunctionInput() {inputName=Frame, inputType=FunctionInput_Scalar}), B=@n1.G)
  FunctionOutput(None=MaterialFunctionCall(X=Add(A=@n2, B=@n3), Y=Add(A=@n4, B=@n5)) {materialFunction=&MakeFloat2}.Result) {outputName=UVs}
 ]}
