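# UnpackBonePosition is an AnimToTexture material function that decodes a bone position from a texture for texture-driven animation.
# It samples the Tex input at the supplied UVs as linear color.
# The RGB value is multiplied by SizeBBox and offset by MinBBox, remapping normalized texture data into the bounding box range as the Position output.
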MaterialFunction /AnimToTexture/Materials/MaterialFunctions/UnpackBonePosition
graph {nodes [
  n1 = TextureSample(
   UVs=FunctionInput() {inputName=UVs, inputType=FunctionInput_Vector2}
   Tex=FunctionInput() {inputName=Tex, inputType=FunctionInput_Texture2D, sortPriority=1}
  ) {samplerType=SAMPLERTYPE_LinearColor, texture=/AnimToTexture/Characters/Mannequin/Textures/BoneAnimation/TX_BoneRotation}
  n2 = FunctionInput() {inputName=SizeBBox, sortPriority=2}
  FunctionOutput(None=Add(A=Multiply(A=@n1.RGB, B=@n2) {constB=1000}, B=FunctionInput() {inputName=MinBBox, sortPriority=3})) {outputName=Position, sortPriority=1}
 ]}
