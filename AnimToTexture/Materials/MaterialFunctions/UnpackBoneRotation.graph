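# Material function in AnimToTexture that decodes a bone rotation stored in a texture as an axis and an angle.
# It samples the input Texture2D at the input UVs in linear color, remaps RGB from unsigned to signed range and normalizes it as the Axis output.
# The alpha channel passes straight through as the Angle output, with no scaling applied.
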MaterialFunction /AnimToTexture/Materials/MaterialFunctions/UnpackBoneRotation
graph {nodes [
  n1 = TextureSample(
   UVs=FunctionInput() {inputName=UVs, inputType=FunctionInput_Vector2}
   Tex=FunctionInput(Preview=TextureObject()) {inputName=Tex, inputType=FunctionInput_Texture2D, sortPriority=1}
  ) {samplerType=SAMPLERTYPE_LinearColor, texture=/AnimToTexture/Characters/Mannequin/Textures/BoneAnimation/TX_BoneWeight}
  n2 = Multiply(A=@n1.RGB) {constB=2}
  FunctionOutput(None=Normalize(VectorInput=Subtract(A=@n2))) {outputName=Axis}
  FunctionOutput(None=@n1.A) {outputName=Angle, sortPriority=1}
 ]}
