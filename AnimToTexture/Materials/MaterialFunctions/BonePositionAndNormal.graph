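# Material function for AnimToTexture that outputs a position offset and tangent-space normal for one bone, animating vertices from baked bone textures.
# It samples bone position and rotation at the given frame shifted by one, with frame zero supplying the bone's rest pivot.
# Local vertex position rotates about that pivot, adds bone translation, then converts to world space relative to the object pivot.
# The pre-skinned normal receives the same rotation.
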MaterialFunction /AnimToTexture/Materials/MaterialFunctions/BonePositionAndNormal
graph {nodes [
  n1 = FunctionInput(Preview=TextureObject()) {inputName=BonePositionTexture, inputType=FunctionInput_Texture2D, sortPriority=2}
  n2 = TextureProperty(Tex=@n1)
  n3 = FunctionInput() {inputName=BoneIndex, inputType=FunctionInput_Scalar, sortPriority=1}
  n4 = MaterialFunctionCall!(Frame=Add(A=FunctionInput() {inputName=Frame, inputType=FunctionInput_Scalar}), BoneIndex=@n3, TextureSize=@n2) {materialFunction=&/BoneIndexUVs}
  n5 = MaterialFunctionCall!(UVs=@n4.UVs, Tex=FunctionInput(Preview=TextureObject()) {inputName=BoneRotationTexture, inputType=FunctionInput_Texture2D, sortPriority=3}) {materialFunction=&/UnpackBoneRotation}
  n6 = MaterialFunctionCall(Frame=Constant(), BoneIndex=@n3, TextureSize=@n2) {materialFunction=&BoneIndexUVs}
  n7 = FunctionInput() {inputName=SizeBBox, sortPriority=4}
  n8 = FunctionInput() {inputName=MinBBox, sortPriority=5}
  n9 = Add(
   A=MaterialFunctionCall!(UVs=@n4.UVs, Tex=@n1, SizeBBox=@n7, MinBBox=@n8) {materialFunction=&/UnpackBonePosition}.Position
   B=RotateAboutAxis(
    NormalizedRotationAxis=@n5.Axis
    RotationAngle=@n5.Angle
    PivotPoint=MaterialFunctionCall(UVs=@n6.UVs, Tex=@n1, SizeBBox=@n7, MinBBox=@n8) {materialFunction=&UnpackBonePosition}.Position
    Position=LocalPosition().XYZ
   )
  )
  FunctionOutput(None=Subtract(
    A=TransformPosition(None=@n9) {transformType=TRANSFORMPOSSOURCE_World}
    B=MaterialFunctionCall() {materialFunction=&ObjectPivotPoint}.Object Pivot Location
   )) {outputName=Position, sortPriority=1}
  n10 = PreSkinnedNormal()
  n11 = Add(A=RotateAboutAxis(NormalizedRotationAxis=@n5.Axis, RotationAngle=@n5.Angle, PivotPoint=Constant(), Position=@n10), B=@n10) {desc=not sure why this is needed}
  FunctionOutput(None=Transform(None=Normalize(VectorInput=@n11)) {transformSourceType=TRANSFORMSOURCE_Local, transformType=TRANSFORM_Tangent}) {outputName=Normal}
 ]}
