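# Material function that unpacks per-vertex bone indices and weights from an AnimToTexture bone weight texture.
# VertexUVs locates the vertex's texel using texture size and RowsPerFrame, treating frame 0 as indices and frame 1 as weights.
# Index RGBA is multiplied by NumBones and rounded, while weight RGBA passes through, giving four channels each.
# A Custom node computing texel-center UVs from a fixed vertex index is left unconnected.
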MaterialFunction /AnimToTexture/Materials/MaterialFunctions/UnpackBoneWeights
graph {nodes [
  MaterialExpressionConstant_1 = Constant() {r=7171}
  Custom!(TextureSize, VertexIndex=@MaterialExpressionConstant_1) {
   code="float Width = TextureSize.x;\r\nfloat Height = TextureSize.y;\r\n\r\nfloat U = (0.5 / Width) + int(VertexIndex % Width) / Width;\r\nfloat V = (0.5 / Height) + int(VertexIndex / Width) / Height;\r\n\r\nreturn float2(U, V);"
   outputType=CMOT_Float2
   inputs [
    {inputName=TextureSize, input {expression=null}}
    {inputName=VertexIndex, input {expression=&self:MaterialExpressionConstant_1}}
   ]
  }
  n1 = FunctionInput(Preview=TextureObject()) {inputName=Tex, inputType=FunctionInput_Texture2D}
  n2 = TextureProperty(Tex=@n1)
  n3 = FunctionInput() {inputName=RowsPerFrame, inputType=FunctionInput_Scalar, sortPriority=1}
  n4 = TextureSample(
   UVs=MaterialFunctionCall!(Frame=Constant(), TextureSize=@n2, RowsPerFrame=@n3) {materialFunction=&/VertexUVs}.UVs
   Tex=@n1
  ) {samplerType=SAMPLERTYPE_LinearColor, texture=/AnimToTexture/Characters/Mannequin/Textures/BoneAnimation/TX_BoneWeight}
  n5 = FunctionInput() {inputName=NumBones, inputType=FunctionInput_Scalar, sortPriority=2}
  FunctionOutput(None=Round(None=Multiply(A=@n4.RGBA, B=@n5) {constB=255})) {outputName=Indices}
  n6 = Constant() {r=1}
  FunctionOutput(None=TextureSample(
    UVs=MaterialFunctionCall(Frame=@n6, TextureSize=@n2, RowsPerFrame=@n3) {materialFunction=&VertexUVs}.UVs
    Tex=@n1
   ) {samplerType=SAMPLERTYPE_LinearColor, texture=&TX_BoneWeight}.RGBA) {outputName=Weights, sortPriority=1}
 ]}
